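; Alley Town welcome sign prop with a gong that reacts to being hit by playing a sound and animation.
; Server damage on the cylinder component finds the hit side via dot product and sends the instigating player controller an Event.Gong.Hit gameplay event.
; PlaySound runs its effects only without authority.
; The animation is material-driven: it sets direction and start time parameters, enabling World Position Offset evaluation until a restartable cleanup timer disables it.
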(event EventOnDamageServer
  (params (Damage :type Float%20%28single-precision%29) (DamageTags :type Gameplay%20Tag%20Container%20Structure%20%28by%20ref%29) (Momentum :type Vector :default "0, 0, 0") (HitInfo :type Hit%20Result%20Structure%20%28by%20ref%29) (InstigatedBy :type Controller%20Object%20Reference) (DamageCauser :type Actor%20Object%20Reference) (EffectContext :type Gameplay%20Effect%20Context%20Handle%20Structure))
  (bind (_bblockinghit _binitialoverlap _time _distance _location _impactpoint _normal _impactnormal _physmat _hitactor _hitcomponent) (Collision|BreakHitResult HitInfo))
  (bind _returnvalue (Fortnite|GetFortPlayerControllerfromActor InstigatedBy))
  (if (== (Variables|Default|GetCylinder) _hitcomponent)
    (CallFunction|PlaySound (< (Math|Vector|DotProduct _impactnormal (Transformation|GetActorRightVector)) 0))
    (Utilities|IsValid _returnvalue
      (:Is%20Valid
        (Ability|SendGameplayEventtoActor _returnvalue "(TagName=\"Event.Gong.Hit\")" (Utilities|Struct|MakeGameplayEventData "(TagName=\"\")" self))))))

(event Custom|PlaySound
  (params (IsFrontHit :type Boolean))
  (Networking|SwitchHasAuthority
    (:Remote
      (Utilities|IsValid self
        (:Is%20Valid
          (Audio|SpawnSoundatLocation (Variables|Default|GetGongSoundCue) (Transformation|GetActorLocation))
          (Default|PlayHitAnimation IsFrontHit))))))

(fn CleanUpHitAnimation
  (Rendering|LOD|SetEvaluateWorldPositionOffset (Variables|Debug|GetStaticMeshComponent) false :!class /Script/Engine.StaticMeshComponent))

(fn PlayHitAnimation
  (params (IsFront :type Boolean))
  (bind _returnvalue (Building|GetBuildingMeshComponent))
  (Rendering|LOD|SetEvaluateWorldPositionOffset (Variables|Debug|GetStaticMeshComponent) true :!class /Script/Engine.StaticMeshComponent)
  (Rendering|Material|SetScalarParameterValueonMaterials _returnvalue "IsForward" (Math|Conversions|ToFloat(Boolean) IsFront))
  (Rendering|Material|SetScalarParameterValueonMaterials _returnvalue "StartTime" (Utilities|Time|GetTimeSeconds))
  (Utilities|Time|ClearandInvalidateTimerbyHandle (Variables|Default|GetgongTimerHandle))
  (bind _returnvalue_2 (Utilities|Time|SetTimerbyEvent (EventDispatchers|CreateEvent :!event CleanUpHitAnimation) 5.0))
  (Variables|Default|SetgongTimerHandle _returnvalue_2))

(fn ConstructionScript
  (|CallParentFunction))
