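# Default Lit material for the light bulbs on the AlleyTown welcome sign, producing an animated, stepped chase-light glow.
# Time is rounded into steps by PanSnap, then drives a panner on UV channel 2 so the bulb mask texture jumps rather than slides.
# The mask lerps from dim to lit and multiplies a tintable, Fresnel-darkened light color, boosted for emissive intensity.
# A normal map parameter uses the sign's texture.
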Material /Hermes/Props/AlleyTown/AlleyTown_WelcomeSign_A/Materials/M_AlleyTown_WelcomeSign_Bulbs
graph {
 properties {bAutomaticallySetUsageInEditor=false, usedShadingModels=Default Lit}
 nodes [
  Constant() {r=0.7}
  Constant() {r=0.1}
  n1 = Time()
  Frac(None=@n1)
  n2 = VectorParameter() {parameterName=Light Emissive Color, defaultValue=<rgba 3.0531 1.031 0.3657 1>}
  Multiply(A=Saturate(None=@n2.RGB)) {constB=0}
  n3 = Fresnel() {baseReflectFraction=0, exponent=1}
  n4 = ScalarParameter() {defaultValue=9, parameterName=PanSnap}
  n5 = Round(None=Multiply(A=@n1, B=@n4) {constB=4})
  n6 = ScalarParameter() {defaultValue=-3, parameterName=Pan X speed}
  n7 = Constant()
  n8 = TextureSample(UVs=Panner(Coordinate=TextureCoordinate() {coordinateIndex=2}, Time=Divide(A=@n5, B=@n4) {constB=4}, Speed=AppendVector(A=@n6, B=@n7)) {speedX=-1}) {samplerType=SAMPLERTYPE_LinearColor, texture=/Hermes/Props/AlleyTown/AlleyTown_WelcomeSign_A/Textures/T_AlleyTown_WelcomeSignBulbs_A_A_E}
  Multiply(A=Multiply(A=LinearInterpolate(A=@n2.RGB, Alpha=@n3) {constB=0}, B=LinearInterpolate(Alpha=@n8.RGB) {constA=0.025})) {constB=20}
  TextureSampleParameter2D() {
   parameterName=Normal
   samplerType=SAMPLERTYPE_Normal
   texture=/Hermes/Props/AlleyTown/AlleyTown_WelcomeSign_A/Textures/T_AlleyTown_WelcomeSign_A_A_N
  }
 ]
}
